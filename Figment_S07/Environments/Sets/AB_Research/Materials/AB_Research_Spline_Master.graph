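# Master spline-mesh material for the AB_Research set, layering baked runway textures with puddles and Athena snow.
# Puddles darken base color and zero roughness, masked by inverted vertex red plus diffuse alpha, and by vertex green.
# Snow masking adds diffuse alpha to a height blend of inverted vertex red and depth-blended world-space noise.
# Snow normals blend angle-corrected over baked normals, while Low quality skips the snow layer.
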Material /Figment_S07/Environments/Sets/AB_Research/Materials/AB_Research_Spline_Master
graph {
 properties {bUseMaterialAttributes=true, bUsedWithSplineMeshes=true, pixelDepthOffsetMode=PDOM_Legacy, usedShadingModels=Default Lit}
 nodes [
  n1 = TextureSampleParameter2D() {group=Baked Textures, parameterName=Baked Diffuse, texture=/Figment_S07/Environments/Sets/AB_Research/Textures/T_Runway_D}
  n2 = Constant()
  n3 = VertexColor() {desc=Red - Snow // Green - Puddle}
  n4 = OneMinus(None=@n3.R)
  n5 = Add(A=@n1.A, B=@n4)
  n6 = Multiply(
   A=ScalarParameter() {defaultValue=0.8, group=Puddles, parameterName=Puddle Opacity}
   B=MaterialFunctionCall(In=Clamp(None=@n5), Contrast=ScalarParameter() {defaultValue=1.3289, group=Puddles, parameterName=Puddle Falloff}) {materialFunction=&CheapContrast}.Result
  )
  n7 = Multiply(A=ScalarParameter() {defaultValue=0.8, group=Puddles, parameterName=Baked Puddle Opacity}, B=@n3.G)
  n8 = TextureSampleParameter2D() {
   group=Baked Textures
   parameterName=Baked Spec
   samplerType=SAMPLERTYPE_Masks
   texture=/Figment_S07/Environments/Sets/AB_Research/Textures/T_Runway_S
  }
  n9 = Constant()
  n10 = TextureSampleParameter2D() {
   group=Baked Textures
   parameterName=Baked Normal
   samplerType=SAMPLERTYPE_Normal
   texture=/Figment_S07/Environments/Sets/AB_Research/Textures/T_Runway_N
  }
  n11 = MakeMaterialAttributes!(
   BaseColor=LinearInterpolate(A=LinearInterpolate(A=@n1.RGB, B=@n2, Alpha=@n6), B=@n2, Alpha=@n7)
   Metallic=@n8.G
   Specular=@n8.G
   Roughness=LinearInterpolate(A=LinearInterpolate(A=@n8.B, B=@n9, Alpha=@n6), B=Constant(), Alpha=@n7)
   Anisotropy
   EmissiveColor
   Opacity
   OpacityMask
   Normal=MaterialFunctionCall(Normal=@n10.RGB, Flatness=ScalarParameter() {group=Baked Textures, parameterName=Baked Normal Flatness}) {materialFunction=&FlattenNormal}.Result
   Tangent
   WorldPositionOffset
   SubsurfaceColor
   ClearCoat
   ClearCoatRoughness
   AmbientOcclusion
   Refraction
   CustomizedUVs_0
   CustomizedUVs_1
   CustomizedUVs_2
   CustomizedUVs_3
   CustomizedUVs_4
   CustomizedUVs_5
   CustomizedUVs_6
   CustomizedUVs_7
   PixelDepthOffset
   ShadingModel
   Displacement
  )
  n12 = GetMaterialAttributes(None=MaterialFunctionCall!(UseLandscapeCoordsOnSwitch) {materialFunction=/Game/Athena/Environments/Landscape/MaterialFunctions/Snow/MF_Athena_Snow_01}.Material) {attributeGetTypes [
    #1
    #2
    #3
    #4
    #5
   ]}
  n13 = ComponentMask(None=Divide(A=WorldPosition().XYZ, B=ScalarParameter() {defaultValue=500, group=Graphical Blend Mask, parameterName=B})) {g=true, r=true}
  n14 = Divide(A=@n13, B=ScalarParameter() {defaultValue=3, group=Graphical Blend Mask, parameterName=A})
  n15 = Divide(
   A=MaterialFunctionCall!(World Position) {materialFunction=/Engine/Functions/Engine_MaterialFunctions02/Utility/DepthFromWorldPosition}.Result
   B=ScalarParameter() {defaultValue=5000, group=Graphical Blend Mask, parameterName=depth}
  )
  n16 = Multiply(
   A=LinearInterpolate(
    A=TextureSample(UVs=@n13) {samplerType=SAMPLERTYPE_LinearGrayscale, texture=/Game/Global/Textures/Default/Noise/T_SwirlNoise_H2}.R
    B=TextureSample(UVs=@n14) {samplerType=SAMPLERTYPE_LinearGrayscale, texture=&T_SwirlNoise_H2}.RGB
    Alpha=Saturate(None=@n15)
   )
   B=ScalarParameter() {defaultValue=0.2, group=Graphical Blend Mask, parameterName=SnowEdgeContrast}
  )
  n17 = ScalarParameter() {defaultValue=6.5231, group=Graphical Blend Mask, parameterName=Snow Contrast}
  n18 = Clamp(None=Add(
    A=@n1.A
    B=MaterialFunctionCall(Transition Phase=@n4, Height Texture=@n16, Contrast=@n17) {materialFunction=&HeightLerp}.Alpha
   ))
  n19 = MaterialFunctionCall!(
   Material=MaterialFunctionCall!(
    Base Material=@n11
    Top Material=MakeMaterialAttributes(BaseColor=@n12.Base Color, Metallic=@n12.Base Color, Specular=@n12.Base Color, Roughness=@n12.Base Color)
    Alpha=@n18
   ) {materialFunction=/Engine/Functions/MaterialLayerFunctions/MatLayerBlend_Standard}.Blended Material
   Additional Normal=MaterialFunctionCall(Normal=@n12.Normal, Flatness=ScalarParameter() {group=Snow, parameterName=Snow Normal Flatness}) {materialFunction=&FlattenNormal}.Result
   Mask=@n18
  ) {
   desc=Blending Baked Normal with Snow Normal
   materialFunction=/Engine/Functions/MaterialLayerFunctions/MatLayerBlend_BlendAngleCorrectedNormals
  }
  QualitySwitch(Default=@n19.Blended Material, Low=@n11, High=@n19.Blended Material, Medium=@n19.Blended Material)
  StaticBool() {value=true}
 ]
}
